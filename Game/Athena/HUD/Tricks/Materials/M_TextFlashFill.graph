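# UI material for trick text in the Athena HUD, giving it a gold-to-white flash fill with a shine.
# The WhiteFill parameter lerps the base color from golden yellow to white.
# MF_Shine generates a highlight controlled by ShineOffset and ShineRotation parameters.
# Overlay and Linear Dodge nodes each blend that shine over the fill color, with no link to a material output shown.
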Material /Game/Athena/HUD/Tricks/Materials/M_TextFlashFill
graph {
 properties {materialDomain=MD_UI, pixelDepthOffsetMode=PDOM_Legacy, usedShadingModels=Default Lit}
 nodes [
  n1 = LinearInterpolate(
   A=Constant3Vector() {constant=<rgba 1 0.7758 0 0>}
   B=Constant3Vector() {constant=<rgba 1 1 1 0>}
   Alpha=ScalarParameter() {parameterName=WhiteFill}
  )
  n2 = MaterialFunctionCall!(
   ShineOffset=ScalarParameter() {defaultValue=-1, group=Shine, parameterName=ShineOffset, sliderMax=2, sliderMin=-2}
   Rotation=ScalarParameter() {defaultValue=0.0476, group=Shine, parameterName=ShineRotation, sliderMax=1}
   RotationCenter
   UVs
  ) {materialFunction=/Game/Athena/HUD/Materials/Victory/MF_Shine}
  MaterialFunctionCall(Base=@n1, Blend=@n2.Result) {materialFunction=&Blend_Overlay}
  MaterialFunctionCall!(Blend=@n2.Result, Base=@n1) {materialFunction=/Engine/Functions/Engine_MaterialFunctions03/Blends/Blend_LinearDodge}
 ]
}
