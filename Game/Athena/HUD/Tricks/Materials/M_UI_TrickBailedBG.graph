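# Translucent UI material for the background graphic of the "trick bailed" HUD element, built from a spray, a plate and drips packed into the RGB channels of one texture.
# Each channel is masked by an antialiased threshold over a UV gradient, with amount parameters controlling each reveal and angle parameters rotating the spray and plate wipes.
# The combined plate and drip mask blends spray color into plate color, and its max with the spray mask forms the overall shape.
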Material /Game/Athena/HUD/Tricks/Materials/M_UI_TrickBailedBG
graph {
 properties {blendMode=BLEND_Translucent, materialDomain=MD_UI, pixelDepthOffsetMode=PDOM_Legacy, usedShadingModels=Default Lit}
 nodes [
  n1 = TextureSample() {texture=&/T_UI_TrickBailedBG_Channels}
  n2 = ScalarParameter() {parameterName=PlateAmount}
  n3 = TextureCoordinate()
  n4 = ScalarParameter() {parameterName=PlateAngle}
  n5 = OneMinus(None=ComponentMask(None=MaterialFunctionCall(UVs=@n3, "Rotation Angle (0-1)"=@n4) {materialFunction=&CustomRotator}.Rotated Values) {r=true})
  n6 = ScalarParameter() {parameterName=Fudge}
  n7 = ScalarParameter() {parameterName=DripAmount}
  n8 = OneMinus(None=ComponentMask(None=TextureCoordinate()) {g=true})
  n9 = ScalarParameter() {parameterName=Fudge}
  n10 = Max(
   A=Multiply(
    A=@n1.G
    B=MaterialFunctionCall!(Threshold=@n2, Gradient=@n5, Filter Width Override, Filter Width Fudge Factor=@n6) {materialFunction=/Game/UI/Materials/AAThreshold}.Result
   )
   B=Multiply(
    A=@n1.B
    B=MaterialFunctionCall(Threshold=@n7, Gradient=@n8, Filter Width Fudge Factor=@n9) {materialFunction=&AAThreshold}.Result
   )
  )
  LinearInterpolate(
   A=VectorParameter() {parameterName=BGSprayColor, defaultValue=<rgba 0.2918 0.0027 0.0369 1>}.RGB
   B=VectorParameter() {parameterName=PlateColor, defaultValue=<rgba 0.2705 0 0.4969 1>}.RGB
   Alpha=@n10
  )
  n11 = ScalarParameter() {parameterName=BGSprayAmount}
  n12 = ComponentMask(None=MaterialFunctionCall(UVs=TextureCoordinate(), "Rotation Angle (0-1)"=ScalarParameter() {parameterName=BGSprayAngle}) {materialFunction=&CustomRotator}.Rotated Values) {g=true}
  n13 = ScalarParameter() {parameterName=Fudge}
  Max(
   A=Multiply(
    A=@n1.R
    B=MaterialFunctionCall(Threshold=@n11, Gradient=@n12, Filter Width Fudge Factor=@n13) {materialFunction=&AAThreshold}.Result
   )
   B=@n10
  )
 ]
}
