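# Translucent UI material for the "Bailed" trick HUD background, defaulting to a bolt backplate shape.
# A two-color gradient comes from lerping Color1 to Color2 along the X channel of UVs rotated by GradientAngle.
# The shape texture, apparently a distance field, is scaled about its center by Scale.
# AAThreshold cuts its red channel at the Edge value to produce an anti-aliased silhouette.
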Material /Game/Athena/HUD/Tricks/Materials/M_UI_Trick_BailedBG
graph {
 properties {blendMode=BLEND_Translucent, materialDomain=MD_UI, pixelDepthOffsetMode=PDOM_Legacy, usedShadingModels=Default Lit}
 nodes [
  n1 = TextureCoordinate()
  n2 = ScalarParameter() {parameterName=GradientAngle}
  LinearInterpolate(
   A=VectorParameter() {parameterName=Color1, defaultValue=<rgba 1 1 1 1>}.RGB
   B=VectorParameter() {parameterName=Color2, defaultValue=<rgba 1 1 1 1>}.RGB
   Alpha=ComponentMask(None=MaterialFunctionCall(UVs=@n1, "Rotation Angle (0-1)"=@n2) {materialFunction=&CustomRotator}.Rotated Values) {r=true}
  )
  n3 = ScalarParameter() {defaultValue=1, parameterName=Scale}
  MaterialFunctionCall!(
   Threshold=ScalarParameter() {defaultValue=0.2, parameterName=Edge, sliderMax=1}
   Gradient=TextureSampleParameter2D(UVs=MaterialFunctionCall(Texture Scale=@n3) {materialFunction=&ScaleUVsByCenter}.UVs) {parameterName=Texture, texture=&/T_UI_BoltBackplate_DF}.R
   Filter Width Override
   Filter Width Fudge Factor
  ) {materialFunction=/Game/UI/Materials/AAThreshold}
 ]
}
