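# Trick HUD widget that shows one callout label as a right-justified, sheared text block with a material outline and drop shadow.
# SetCallout uppercases the incoming text, applies it to the Callout text block and plays the Intro animation.
# Intro is a quick pop-in that fades the text in while it shrinks from oversized, dips below normal scale and settles.
# The event graph is empty, so SetCallout must be called externally.
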WidgetBlueprint /Game/Athena/HUD/Tricks/CalloutEntry
animations [Intro {
  binding_count=1
  display_label=""
  length_seconds=0.1167
  object=&self:Intro
  sequence {
   bindings [Callout {
     id=#1
     display_name=CommonTextBlock_Callout
     possessed_class=Border
     tracks [
      {
       type=MovieSceneFloatTrack
       display_name=Render Opacity
       sections [{type=MovieSceneFloatSection, end_frame=14, start_frame=0, channels [None {type=MovieSceneScriptingFloatChannel, keys [[0,0],[14,1]]}]}]
      }
      {
       type=MovieScene2DTransformTrack
       sections [{
         type=MovieScene2DTransformSection
         end_frame=14
         start_frame=0
         channels [
          Translation.X {type=MovieSceneScriptingFloatChannel}
          Translation.Y {type=MovieSceneScriptingFloatChannel}
          Angle {type=MovieSceneScriptingFloatChannel}
          Scale.X {type=MovieSceneScriptingFloatChannel, keys [[0,2],[12,0.8],[14,1]]}
          Scale.Y {type=MovieSceneScriptingFloatChannel, keys [[0,2],[12,0.8],[14,1]]}
          Shear.X {type=MovieSceneScriptingFloatChannel}
          Shear.Y {type=MovieSceneScriptingFloatChannel}
         ]
        }]
      }
     ]
    }]
   display_rate {numerator=120}
  }
 }]
graphs [
 EventGraph {links [], nodes []}
 SetCallout {nodes [
   n1 = SetCallout!() -> (then, InText:Text) {is=title K2Node_FunctionEntry}
   > "SetText (Text)"(self=Get Callout() -> (Callout:Common Text Object Reference), InText=TextToUpper!(InText:"Text (by ref)"=@n1) -> (ReturnValue:Text))
   > PlayAnimation(InAnimation=Get Intro())
  ]}
]
variables []
widgets {
 info {inheritedWidgetCount=0, parentClass=&UserWidget, rootWidgetClass=&CommonTextBlock, widgetCount=1}
 widgets [Callout {
   widgetClassPath=&CommonTextBlock
   bIsVariable=true
   widget=Callout
   properties {
    justification=Right
    style=/Game/Athena/HUD/Tricks/Materials/TextStyle-TrickEntryCallout.TextStyle-TrickEntryCallout_C
    text="CALLOUT NAME!"
    font {
     bForceMonospaced=false
     bMaterialIsStencil=false
     fontMaterial=null
     fontObject=&Burbank
     letterSpacing=0
     monospacedWidth=1
     size=24
     skewAmount=0
     typefaceFontName=Burbank-Black
     outlineSettings {
      bApplyOutlineToDropShadows=false
      bMiteredCorners=false
      bSeparateFillAlpha=false
      outlineMaterial=/Game/Athena/HUD/Tricks/Materials/M_TrickText_Callout
      outlineSize=1
      outlineColor=<rgba 1 1 1 1>
     }
    }
    renderTransform {angle=0, scale=<1 1>, shear=<-8 0>, translation=<0 0>}
    shadowColorAndOpacity=<rgba 0.1779 0.0262 0.003 1>
    shadowOffset=<4 3>
   }
  }]
}
