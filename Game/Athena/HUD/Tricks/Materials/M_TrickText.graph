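# UI-domain material for trick text in the Athena HUD, producing a gradient fill with a shine highlight.
# A rotatable UV's green channel samples the TrickScoreOutline curve atlas row as the gradient.
# The rotation pivot reuses the red channel of RotationCenter for both axes.
# MF_Shine, driven by offset and rotation parameters and scaled only by ShineColor's alpha, is color-dodge blended over the gradient.
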Material /Game/Athena/HUD/Tricks/Materials/M_TrickText
graph {
 properties {materialDomain=MD_UI, pixelDepthOffsetMode=PDOM_Legacy, usedShadingModels=Default Lit}
 nodes [
  n1 = ScalarParameter() {defaultValue=-1, group=Shine, parameterName=ShineOffset, sliderMax=2, sliderMin=-2}
  n2 = ScalarParameter() {defaultValue=0.0476, group=Shine, parameterName=ShineRotation, sliderMax=1}
  n3 = TextureCoordinate()
  n4 = VectorParameter() {parameterName=RotationCenter, defaultValue=<rgba 0.5 0.5 0 1>}
  n5 = AppendVector(A=@n4.R, B=@n4.R)
  n6 = ScalarParameter() {parameterName=Rotation}
  n7 = CurveAtlasRowParameter(CurveTime=ComponentMask(None=MaterialFunctionCall(UVs=@n3, Rotation Center=@n5, "Rotation Angle (0-1)"=@n6) {materialFunction=&CustomRotator}.Rotated Values) {g=true}) {
   atlas=&/NewCurveLinearColorAtlas
   curve=&/TrickScoreOutline
   defaultValue=0.0312
   parameterName=CurveParams
  }
  MaterialFunctionCall!(
   Blend=Multiply(
    A=VectorParameter() {group=Shine, parameterName=ShineColor, defaultValue=<rgba 1 1 1 1>}.A
    B=MaterialFunctionCall!(ShineOffset=@n1, Rotation=@n2, RotationCenter, UVs) {materialFunction=/Game/Athena/HUD/Materials/Victory/MF_Shine}.Result
   )
   Base=@n7
  ) {materialFunction=/Engine/Functions/Engine_MaterialFunctions03/Blends/Blend_ColorDodge}
 ]
}
