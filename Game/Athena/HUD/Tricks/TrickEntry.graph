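# TrickEntry is a HUD widget that displays a single trick name as right-aligned, sheared Burbank text with a flash-fill font material.
# SetTrick uppercases the incoming text, sets it on TrickNameText, and plays the Intro animation.
# Intro slides the text in from the right with a slight overshoot while stretching it vertically and fading it in.
# It also animates the font material's WhiteFill and ShineOffset parameters for a flash and shine sweep.
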WidgetBlueprint /Game/Athena/HUD/Tricks/TrickEntry
animations [Intro {
  binding_count=1
  display_label=""
  length_seconds=0.6083
  object=&self:Intro
  sequence {
   bindings [TrickNameText {
     id=#1
     possessed_class=CommonTextBlock
     tracks [
      {
       type=MovieScene2DTransformTrack
       sections [{
         type=MovieScene2DTransformSection
         end_frame=38
         start_frame=0
         channels [
          Translation.X {type=MovieSceneScriptingFloatChannel, keys [[0,503.2865],[18,-45.5755],[21,0],[25,0]]}
          Translation.Y {type=MovieSceneScriptingFloatChannel}
          Angle {type=MovieSceneScriptingFloatChannel}
          Scale.X {type=MovieSceneScriptingFloatChannel}
          Scale.Y {type=MovieSceneScriptingFloatChannel, keys [[0,0.5754],[14.086,1]]}
          Shear.X {type=MovieSceneScriptingFloatChannel}
          Shear.Y {type=MovieSceneScriptingFloatChannel}
         ]
        }]
      }
      {
       type=MovieSceneWidgetMaterialTrack
       display_name=Font.FontMaterial
       sections [{
         type=MovieSceneParameterSection
         end_frame=73
         start_frame=0
         channels [
          WhiteFill {type=MovieSceneScriptingFloatChannel, keys [[16.294,0],[25,1],[41.072,0]]}
          ShineOffset {type=MovieSceneScriptingFloatChannel, keys [[18,-0.2157],[49.168,1.3265]]}
          ShineRotation {type=MovieSceneScriptingFloatChannel}
         ]
        }]
      }
      {
       type=MovieSceneFloatTrack
       display_name=Render Opacity
       sections [{type=MovieSceneFloatSection, end_frame=72, start_frame=0, channels [None {type=MovieSceneScriptingFloatChannel, keys [[0,0],[14.194,1]]}]}]
      }
     ]
    }]
   display_rate {numerator=120}
  }
 }]
graphs [
 EventGraph {links [], nodes []}
 SetTrick {nodes [
   n1 = SetTrick!() -> (then, InText:Text) {is=title K2Node_FunctionEntry}
   > "SetText (Text)"(
    self=Get TrickNameText() -> (TrickNameText:Common Text Object Reference)
    InText=TextToUpper!(InText:"Text (by ref)"=@n1) -> (ReturnValue:Text)
   )
   > PlayAnimation(InAnimation=Get Intro())
  ]}
]
variables []
widgets {
 info {inheritedWidgetCount=0, parentClass=&UserWidget, rootWidgetClass=&CommonTextBlock, widgetCount=1}
 widgets [TrickNameText {
   widgetClassPath=&CommonTextBlock
   bIsVariable=true
   widget=TrickNameText
   properties {
    justification=Right
    text="TRICK NAME!"
    textTransformPolicy=ToUpper
    font {
     bForceMonospaced=false
     bMaterialIsStencil=false
     fontMaterial=/Game/Athena/HUD/Tricks/Materials/M_TextFlashFill_Inst
     fontObject=&Burbank
     letterSpacing=0
     monospacedWidth=1
     size=18
     skewAmount=0
     typefaceFontName=Burbank-PlayerPower
     outlineSettings {
      bApplyOutlineToDropShadows=false
      bMiteredCorners=false
      bSeparateFillAlpha=false
      outlineMaterial=null
      outlineSize=1
      outlineColor=<rgba 0.6584 0.1746 0.0137 1>
     }
    }
    renderTransform {angle=0, scale=<1 1>, shear=<-12 0>, translation=<0 0>}
   }
  }]
}
