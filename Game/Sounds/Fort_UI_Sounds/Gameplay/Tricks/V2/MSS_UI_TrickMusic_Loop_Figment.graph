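# Stereo one-shot MetaSound for Figment-themed trick-sequence music, playing an intro and then the music loop.
# OnTrick and OnFinished open gates on a repeating trigger, deferring each event to the next pulse.
# Tricks play random stingers; finishing stops the loop, plays the ending and reports completion when it ends.
# The trick-counted layer player reads an empty array, leaving HitHatLayers unused, and MusicBPM is unconnected.
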MetaSoundSource /Game/Sounds/Fort_UI_Sounds/Gameplay/Tricks/V2/MSS_UI_TrickMusic_Loop_Figment
inputs [
 First Trick {type=WaveAsset}
 HitHatLayers {type=WaveAsset:Array}
 MusicBPM {type=Float}
 MusicFinished {type=WaveAsset}
 MusicLoop {type=WaveAsset}
 OnFinished {type=Trigger}
 OnTrick {type=Trigger}
 TrickStingers {type=WaveAsset:Array}
 UE.Source.OnPlay {type=Trigger}
 Wave Asset {type=WaveAsset}
]
interfaces [UE.OutputFormat.Stereo:1.0, UE.Source.OneShot:1.0, UE.Source:1.1]
outputs [
 "UE.OutputFormat.Stereo.Audio:0" {type=Audio}
 "UE.OutputFormat.Stereo.Audio:1" {type=Audio}
 UE.Source.OneShot.OnFinished {type=Trigger}
]
pages [Default {
  input_defaults {
   First Trick=/Game/Sounds/Fort_UI_Sounds/Gameplay/Tricks/V2/FigmentAudio/TrickUI_SequenceMusicLoop_Intro_09
   MusicBPM=160
   MusicFinished=/Game/Sounds/Fort_UI_Sounds/Gameplay/Tricks/TrickUI_SequenceEnd_Success_07
   MusicLoop=/Game/Sounds/Fort_UI_Sounds/Gameplay/Tricks/V2/FigmentAudio/TrickUI_SequenceMusicLoop_02_OPT_02
   OnFinished=null
   OnTrick=null
   UE.Source.OnPlay=null
   Wave Asset=/Game/Sounds/Fort_UI_Sounds/Gameplay/Tricks/TrickUI_SequenceEnd_Fail_08
   HitHatLayers [
    /Game/Sounds/Fort_UI_Sounds/Gameplay/Tricks/V2/Waves/MX_Trick_MC_DB_Hats_Mult1_Loop
    /Game/Sounds/Fort_UI_Sounds/Gameplay/Tricks/V2/Waves/MX_Trick_MC_DB_Hats_Mult2_Loop
    /Game/Sounds/Fort_UI_Sounds/Gameplay/Tricks/V2/Waves/MX_Trick_MC_DB_Hats_Mult3_Loop
   ]
   TrickStingers [
    /Game/Sounds/Fort_UI_Sounds/Gameplay/Tricks/V2/FigmentAudio/TrickUI_NewTrick_06
    /Game/Sounds/Fort_UI_Sounds/Gameplay/Tricks/V2/FigmentAudio/TrickUI_NewTrick_05
    /Game/Sounds/Fort_UI_Sounds/Gameplay/Tricks/V2/FigmentAudio/TrickUI_NewTrick_04
    /Game/Sounds/Fort_UI_Sounds/Gameplay/Tricks/V2/FigmentAudio/TrickUI_NewTrick_03
   ]
  }
  nodes [
   ArrayRandomGet_0 = Array.Random Get:WaveAsset:Array(In Array=^TrickStingers)
   WavePlayer_1 = UE.Wave Player:Stereo(Wave Asset=^First Trick, Play=^UE.Source.OnPlay) {also {version [1,1]}}
   n1 = InitVariable.Trigger()
   n2 = VariableDeferredAccessor.Trigger(Variable=@n1.Variable)
   TriggerCounter_0 = UE.Trigger Counter(Reset=^UE.Source.OnPlay, In=@n2.Value)
   n3 = Min.Min:Int32(B=Array.Num:WaveAsset:Array(Array=^TrickStingers).Num, A=@TriggerCounter_0.Count)
   n4 = InitVariable.Trigger()
   n5 = VariableDeferredAccessor.Trigger(Variable=@n4.Variable)
   WavePlayer_2 = UE.Wave Player:Stereo(
    Wave Asset=Array.Get:WaveAsset:Array(
     Trigger=UE.Reroute(Value=@TriggerCounter_0.On Trigger).Value
     Index=UE.Subtract:Int32(PrimaryOperand=@n3.Value) {also {defaults {AdditionalOperands=1}}}.Out
    ) {also {defaults {Array []}}}.Element
    Play=@TriggerCounter_0.On Trigger
    Stop=@n5.Value
   ) {also {defaults {Loop="<MetaSoundBuilderResult.SUCCEEDED: 0>"}, version [1,1]}}
   n6 = VariableDeferredAccessor.Trigger(Variable=@n2.Variable)
   ArrayRandomGet_1 = Array.Random Get:WaveAsset:Array(In Array=^TrickStingers, Next=@n6.Value) {also {defaults {Weights []}}}
   WavePlayer_3 = UE.Wave Player:Stereo(Play=@ArrayRandomGet_1.On Next, Wave Asset=@ArrayRandomGet_1.Value) {also {version [1,1]}}
   WavePlayer_4 = UE.Wave Player:Stereo(
    Wave Asset=^MusicLoop
    Play=UE.Reroute(Value=@WavePlayer_1.On Finished).Value
    Stop=VariableDeferredAccessor.Trigger(Variable=@n4.Variable).Value
   ) {also {defaults {Loop="<MetaSoundBuilderResult.SUCCEEDED: 0>"}, version [1,1]}}
   n7 = VariableDeferredAccessor.Trigger(Variable=@n5.Variable)
   WavePlayer_5 = UE.Wave Player:Stereo(Wave Asset=^MusicFinished, Play=@n7.Value) {also {version [1,1]}}
   AudioMixerAudioMixerStereo8_0 = "AudioMixer.Audio Mixer (Stereo, 8)"(
    In 1 R=UE.Reroute(Value=@WavePlayer_1.Out Right).Value
    In 1 L=UE.Reroute(Value=@WavePlayer_1.Out Left).Value
    In 6 L=@WavePlayer_2.Out Left
    In 6 R=@WavePlayer_2.Out Right
    In 3 L=@WavePlayer_3.Out Left
    In 3 R=@WavePlayer_3.Out Right
    In 0 L=@WavePlayer_4.Out Left
    In 0 R=@WavePlayer_4.Out Right
    In 5 L=@WavePlayer_5.Out Left
    In 5 R=@WavePlayer_5.Out Right
   )
   UE.BPMToSeconds()
   InitVariable.Bool()
   InitVariable.Bool()
   n8 = VariableDeferredAccessor.Bool(Variable=InitVariable.Bool().Variable)
   n9 = TriggerCompare.Clamp:Bool(Compare=^OnFinished, A=@n8.Value)
   n10 = VariableAccessor.Trigger(Variable=VariableMutator.Trigger(Variable=InitVariable.Trigger().Variable, Value=UE.TriggerRepeat(Start=^UE.Source.OnPlay).RepeatOut).Variable)
   TriggerControl_1 = UE.Trigger Control(
    Open=@n9.TRUE
    Trigger In=VariableAccessor.Trigger(Variable=@n10.Variable).Value
    Close=VariableDeferredAccessor.Trigger(Variable=@n4.Variable).Value
   )
   VariableMutator.Trigger(Value=@TriggerControl_1.Trigger Out, Variable=@n7.Variable)
   TriggerRouteTriggerRouteBool2_0 = "TriggerRoute.Trigger Route (Bool, 2):Bool"(Set 1=^UE.Source.OnPlay, Set 0=@n9.TRUE) {also {defaults {Value 0="<MetaSoundBuilderResult.SUCCEEDED: 0>"}}}
   n11 = VariableMutator.Bool(Value=@TriggerRouteTriggerRouteBool2_0.Value, Variable=@n8.Variable)
   TriggerControl_0 = UE.Trigger Control(
    Open=TriggerCompare.Clamp:Bool(Compare=^OnTrick, A=VariableAccessor.Bool(Variable=@n11.Variable).Value).TRUE
    Trigger In=@n10.Value
    Close=VariableDeferredAccessor.Trigger(Variable=@n1.Variable).Value
   )
   VariableMutator.Trigger(Value=@TriggerControl_0.Trigger Out, Variable=@n6.Variable)
   WavePlayer_0 = UE.Wave Player:Stereo() {also {version [1,1]}}
  ]
  links [
   {dst="@UE.OutputFormat.Stereo.Audio:0", src=AudioMixerAudioMixerStereo8_0.Out L}
   {dst="@UE.OutputFormat.Stereo.Audio:1", src=AudioMixerAudioMixerStereo8_0.Out R}
   {dst="@UE.Source.OneShot.OnFinished", src=WavePlayer_5.On Finished}
  ]
 }]
properties {
 assetImportData=&self:AssetImportData
 numChannels=2
 outputFormat=Stereo
 sampleRate=48000
 soundClassObject=/Game/Sounds/FortSoundClassesAndModes/Fort_Music_Gameplay
}
