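# HUD trick feed widget on AthenaTrickFeedBase, showing vehicle trick names, score, multiplier and stats during a trick sequence.
# Base-class events fill a capped entry list, restyle the score past a critical threshold, and add random tiered callouts on completion.
# Failed or canceled sequences play a wipeout animation; hiding ends by requesting the next sequence.
# Playlist-aware MetaSound music starts per sequence, triggers trick stingers, and ends with success or fail audio.
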WidgetBlueprint /Game/Athena/HUD/Tricks/TrickFeedView
animations [
 Crit {
  binding_count=1
  display_label=""
  length_seconds=0.4752
  object=&self:Crit
  sequence {
   bindings [Image_Glow {
     id=#1
     possessed_class=Image
     tracks [{
       type=MovieScene2DTransformTrack
       sections [{
         type=MovieScene2DTransformSection
         end_frame=57
         start_frame=0
         channels [
          Translation.X {type=MovieSceneScriptingFloatChannel}
          Translation.Y {type=MovieSceneScriptingFloatChannel}
          Angle {type=MovieSceneScriptingFloatChannel}
          Scale.X {type=MovieSceneScriptingFloatChannel, keys [[0,0],[13.786,1.4],[57.02,1]]}
          Scale.Y {type=MovieSceneScriptingFloatChannel, keys [[0,0],[13.786,1.4],[57.02,1]]}
          Shear.X {type=MovieSceneScriptingFloatChannel}
          Shear.Y {type=MovieSceneScriptingFloatChannel}
         ]
        }]
      }]
    }]
   display_rate {numerator=120}
  }
 }
 MultiplierBump {
  binding_count=1
  display_label=""
  length_seconds=0.2
  object=&self:MultiplierBump
  sequence {
   bindings [Multiplier {
     id=#2
     possessed_class=CommonTextBlock
     tracks [{
       type=MovieScene2DTransformTrack
       sections [{
         type=MovieScene2DTransformSection
         end_frame=24
         start_frame=0
         channels [
          Translation.X {type=MovieSceneScriptingFloatChannel}
          Translation.Y {type=MovieSceneScriptingFloatChannel}
          Angle {type=MovieSceneScriptingFloatChannel}
          Scale.X {type=MovieSceneScriptingFloatChannel, keys [[0,1],[3.086,1.3],[17.992,0.9944],[23.998,1]]}
          Scale.Y {type=MovieSceneScriptingFloatChannel, keys [[0,1],[3.086,1.3],[17.992,0.9944],[23.998,1]]}
          Shear.X {type=MovieSceneScriptingFloatChannel}
          Shear.Y {type=MovieSceneScriptingFloatChannel}
         ]
        }]
      }]
    }]
   display_rate {numerator=120}
  }
 }
 TriangleFlare {
  binding_count=2
  display_label=""
  length_seconds=1.5
  object=&self:TriangleFlare
  sequence {bindings [
    CommonBorder_Flare {
     id=#3
     possessed_class=CommonBorder
     tracks [{
       type=MovieSceneFloatVectorTrack
       display_name=Desired Size Scale
       sections [{
         type=MovieSceneFloatVectorSection
         channels [
          Vector.X {type=MovieSceneScriptingFloatChannel}
          Vector.Y {type=MovieSceneScriptingFloatChannel, keys [[0,0.7],[5.7913,1]]}
         ]
        }]
      }]
    }
    SizeBox_Flares {
     id=#4
     possessed_class=SizeBox
     tracks [
      {
       type=MovieSceneFloatTrack
       display_name=Height Override
       sections [{
         type=MovieSceneFloatSection
         end_frame=15
         start_frame=0
         channels [None {type=MovieSceneScriptingFloatChannel, keys [[0,30.877],[5.7913,52.4973]]}]
        }]
      }
      {
       type=MovieSceneFloatTrack
       display_name=Render Opacity
       sections [{
         type=MovieSceneFloatSection
         end_frame=30
         start_frame=0
         channels [None {type=MovieSceneScriptingFloatChannel, keys [[0,0],[4.1883,1],[29.9997,0]]}]
        }]
      }
     ]
    }
   ]}
 }
 TrickFailed {
  binding_count=6
  display_label=""
  length_seconds=4.4095
  object=&self:TrickFailed
  sequence {
   bindings [
    CommonBorder_TrickContainer {
     id=#5
     possessed_class=CommonBorder
     tracks [{
       type=MovieSceneColorTrack
       display_name=Content Color and Opacity
       sections [{
         type=MovieSceneColorSection
         end_frame=358
         start_frame=0
         channels [
          Color.R {type=MovieSceneScriptingFloatChannel, keys [[0,1],[4.756,0.5],[358.968,0.0052]]}
          Color.G {type=MovieSceneScriptingFloatChannel, keys [[0,1],[4.756,0],[358.968,0]]}
          Color.B {type=MovieSceneScriptingFloatChannel, keys [[0,1],[4.756,0],[358.968,0]]}
          Color.A {type=MovieSceneScriptingFloatChannel, keys [[0,1],[4.756,1],[358.968,0]]}
         ]
        }]
      }]
    }
    Overlay_WipeoutContainer {
     id=#6
     possessed_class=Overlay
     tracks [{
       type=MovieSceneFloatTrack
       display_name=Render Opacity
       sections [{
         type=MovieSceneFloatSection
         end_frame=529
         start_frame=0
         channels [None {type=MovieSceneScriptingFloatChannel, keys [[0,0],[30,1],[300.562,1],[529.134,0]]}]
        }]
      }]
    }
    Image_WipeoutBG {
     id=#7
     possessed_class=Image
     tracks [{
       type=MovieSceneWidgetMaterialTrack
       sections [{
         type=MovieSceneParameterSection
         end_frame=520
         start_frame=20
         channels [
          BGSprayAmount {type=MovieSceneScriptingFloatChannel, keys [[70.752,1.7287],[374.258,-0.1996]]}
          DripAmount {type=MovieSceneScriptingFloatChannel, keys [[123.086,0.9748],[518.752,0.0157]]}
          PlateAmount {type=MovieSceneScriptingFloatChannel, keys [[26.966,1.1577],[206.232,-0.2429]]}
         ]
        }]
      }]
    }
    FailedText {
     id=#8
     possessed_class=CommonTextBlock
     tracks [{
       type=MovieScene2DTransformTrack
       sections [{
         type=MovieScene2DTransformSection
         end_frame=529
         start_frame=0
         channels [
          Translation.X {type=MovieSceneScriptingFloatChannel}
          Translation.Y {type=MovieSceneScriptingFloatChannel}
          Angle {type=MovieSceneScriptingFloatChannel, keys [[0,-29.0122],[25.2,-4.1416],[30.62,-9.7378],[35.216,-6.9149]]}
          Scale.X {type=MovieSceneScriptingFloatChannel, keys [[0,3],[25,0.8],[30.138,1.0171],[35,1],[396.434,1],[412.974,1.1024],[441.182,0]]}
          Scale.Y {type=MovieSceneScriptingFloatChannel, keys [[0,3],[24.994,0.7911],[25,0.8],[30.138,1.0171],[35,1],[396.434,1],[412.974,1.1024],[441.182,0]]}
          Shear.X {type=MovieSceneScriptingFloatChannel}
          Shear.Y {type=MovieSceneScriptingFloatChannel}
         ]
        }]
      }]
    }
    TrickInformation {
     id=#9
     possessed_class=VerticalBox
     tracks [{
       type=MovieScene2DTransformTrack
       sections [{
         type=MovieScene2DTransformSection
         end_frame=125
         start_frame=0
         channels [
          Translation.X {
           type=MovieSceneScriptingFloatChannel
           keys [[0,0],[2.024,-31.9869],[6.534,32.0809],[10.908,-19.5037],[13.172,17.2137],[16.364,-21.4039],[17.134,21.4685],[21.172,-19.9964],[23.036,15.1186],[28.636,-16.1245],[31.532,14.9867],[37.052,-10.7926],[41.352,13.5085],[47.18,-6.1626],[47.722,6.2038],[52.716,-5.0656],[54.536,5.3727],[60.244,-2.4982],[63.198,2.0879],[69.184,-1.3651],[76.358,0]]
          }
          Translation.Y {
           type=MovieSceneScriptingFloatChannel
           keys [[0,0],[2.024,-31.9869],[6.534,32.0809],[11.096,-18.8703],[13.138,18.762],[15.914,-21.9669],[17.134,21.4685],[21.924,-22.1077],[22.968,15.4174],[26.624,-15.3929],[31.532,15.3568],[35.04,-10.1119],[41.352,13.8181],[45.166,-5.5236],[47.722,6.4829],[50.704,-4.4593],[54.536,5.6144],[58.23,-1.9264],[63.198,2.2959],[67.172,-0.8238],[76.358,0]]
          }
          Angle {type=MovieSceneScriptingFloatChannel}
          Scale.X {type=MovieSceneScriptingFloatChannel}
          Scale.Y {type=MovieSceneScriptingFloatChannel}
          Shear.X {type=MovieSceneScriptingFloatChannel}
          Shear.Y {type=MovieSceneScriptingFloatChannel}
         ]
        }]
      }]
    }
    Image_WipeoutPlateBG {
     id=#10
     possessed_class=Image
     tracks [{
       type=MovieSceneWidgetMaterialTrack
       sections [{
         type=MovieSceneParameterSection
         end_frame=529
         start_frame=0
         channels [Edge {type=MovieSceneScriptingFloatChannel, keys [[20.326,1.0112],[31.702,0.1214],[49.656,0.3146],[390.722,0.2875],[436.332,1.3792]]}]
        }]
      }]
    }
   ]
   display_rate {numerator=120}
  }
 }
 TrickComplete {
  binding_count=3
  display_label=""
  length_seconds=0.2294
  object=&self:TrickComplete
  sequence {
   bindings [
    CommonBorder_TrickContainer {
     id=#11
     possessed_class=CommonBorder
     tracks [{
       type=MovieSceneFloatTrack
       display_name=Render Opacity
       sections [{type=MovieSceneFloatSection, end_frame=27, start_frame=0, channels [None {type=MovieSceneScriptingFloatChannel, keys [[0,1],[27.526,0]]}]}]
      }]
    }
    Overlay_0 {
     id=#12
     possessed_class=Overlay
     tracks [{
       type=MovieScene2DTransformTrack
       sections [{
         type=MovieScene2DTransformSection
         end_frame=27
         start_frame=0
         channels [
          Translation.X {type=MovieSceneScriptingFloatChannel, keys [[0,0],[27.526,-81.1312]]}
          Translation.Y {type=MovieSceneScriptingFloatChannel}
          Angle {type=MovieSceneScriptingFloatChannel}
          Scale.X {type=MovieSceneScriptingFloatChannel}
          Scale.Y {type=MovieSceneScriptingFloatChannel}
          Shear.X {type=MovieSceneScriptingFloatChannel}
          Shear.Y {type=MovieSceneScriptingFloatChannel}
         ]
        }]
      }]
    }
    TrickList {
     id=#13
     possessed_class=DynamicEntryBox
     tracks [{
       type=MovieScene2DTransformTrack
       sections [{
         type=MovieScene2DTransformSection
         end_frame=27
         start_frame=0
         channels [
          Translation.X {type=MovieSceneScriptingFloatChannel, keys [[1.704,0],[27.526,-38.0249]]}
          Translation.Y {type=MovieSceneScriptingFloatChannel}
          Angle {type=MovieSceneScriptingFloatChannel}
          Scale.X {type=MovieSceneScriptingFloatChannel}
          Scale.Y {type=MovieSceneScriptingFloatChannel}
          Shear.X {type=MovieSceneScriptingFloatChannel}
          Shear.Y {type=MovieSceneScriptingFloatChannel}
         ]
        }]
      }]
    }
   ]
   display_rate {numerator=120}
  }
 }
 TrickReset {
  binding_count=3
  display_label=""
  length_seconds=0.2294
  object=&self:TrickReset
  sequence {bindings [
    Overlay_0 {
     id=#14
     possessed_class=Overlay
     tracks [{
       type=MovieScene2DTransformTrack
       sections [{
         type=MovieScene2DTransformSection
         end_frame=4
         start_frame=0
         channels [
          Translation.X {type=MovieSceneScriptingFloatChannel, const [0,1]}
          Translation.Y {type=MovieSceneScriptingFloatChannel, const [0,1]}
          Angle {type=MovieSceneScriptingFloatChannel, const [0,1]}
          Scale.X {type=MovieSceneScriptingFloatChannel, const [1,1]}
          Scale.Y {type=MovieSceneScriptingFloatChannel, const [1,1]}
          Shear.X {type=MovieSceneScriptingFloatChannel, const [0,1]}
          Shear.Y {type=MovieSceneScriptingFloatChannel, const [0,1]}
         ]
        }]
      }]
    }
    CommonBorder_TrickContainer {
     id=#15
     possessed_class=CommonBorder
     tracks [
      {
       type=MovieSceneFloatTrack
       display_name=Render Opacity
       sections [{type=MovieSceneFloatSection, end_frame=4, start_frame=0, channels [None {type=MovieSceneScriptingFloatChannel, const [1,1]}]}]
      }
      {
       type=MovieSceneColorTrack
       display_name=Content Color and Opacity
       sections [{
         type=MovieSceneColorSection
         end_frame=4
         start_frame=0
         channels [
          Color.R {type=MovieSceneScriptingFloatChannel, const [1,1]}
          Color.G {type=MovieSceneScriptingFloatChannel, const [1,1]}
          Color.B {type=MovieSceneScriptingFloatChannel, const [1,1]}
          Color.A {type=MovieSceneScriptingFloatChannel, const [1,1]}
         ]
        }]
      }
     ]
    }
    TrickList {
     id=#16
     possessed_class=DynamicEntryBox
     tracks [{
       type=MovieScene2DTransformTrack
       sections [{
         type=MovieScene2DTransformSection
         end_frame=4
         start_frame=0
         channels [
          Translation.X {type=MovieSceneScriptingFloatChannel, const [0,1]}
          Translation.Y {type=MovieSceneScriptingFloatChannel, const [0,1]}
          Angle {type=MovieSceneScriptingFloatChannel, const [0,1]}
          Scale.X {type=MovieSceneScriptingFloatChannel, const [1,1]}
          Scale.Y {type=MovieSceneScriptingFloatChannel, const [1,1]}
          Shear.X {type=MovieSceneScriptingFloatChannel, const [0,1]}
          Shear.Y {type=MovieSceneScriptingFloatChannel, const [0,1]}
         ]
        }]
      }]
    }
   ]}
 }
]
graphs [
 EventGraph {nodes [
   HIdeTrickUI() {is=title K2Node_CustomEvent}
   > PlayAnimation(InAnimation=Get TrickComplete() -> (TrickComplete:Widget Animation Object Reference))
   n1 = CritStyleEvent!() -> (OutputDelegate, then, isCritical:bool=~) {is=title K2Node_CustomEvent}
   > Set isCritical(isCritical:bool=@n1.isCritical) -> (Output_Get:bool)
   > n2 = Branch(Condition=@n1.isCritical)
   > n3 = PlayAnimation(InAnimation=Get Crit() -> (Crit:Widget Animation Object Reference))
   n4 = Get Image_Glow() -> (Image_Glow:Image Object Reference)
   SetVisibility(execute=[@n3.then, @n2.else], self=@n4, InVisibility=Select(Option 0=Hidden, Option 1=HitTestInvisible, Index=@n1.isCritical))
   > SetBrushTintColor!(
    execute
    self:Image Object Reference=@n4
    TintColor:Slate Color Structure=Make Slate Color(SpecifiedColor=Select(Option 0:Linear Color Structure, Option 1:Linear Color Structure=(R=0.4852,G=0.0467,B=0,A=1), Index=@n1.isCritical) -> (ReturnValue:Linear Color Structure))
   ) -> (then)
   n5 = Get TotalScore() -> (TotalScore:Common Numeric Text Block Object Reference)
   n6 = Break Slate Font Info!(SlateFontInfo:"Slate Font Info Structure (by ref)"=Get Font!(self:Text Object Reference=@n5) -> (Font:Slate Font Info Structure)) -> (
    FontObject:Object Reference
    FontMaterial:Object Reference
    OutlineSettings:Font Outline Settings Structure
    TypefaceFontName:Name
    Size:float=~
    LetterSpacing:int
    SkewAmount:float=~
    bForceMonospaced:bool=~
    bMaterialIsStencil:bool=~
    MonospacedWidth:float=~
   ) {is=title K2Node_BreakStruct}
   n7 = Select(Option 0:int=30, Option 1:int=48, Index=@n1.isCritical) -> (ReturnValue:int)
   n8 = Break Font Outline Settings!(FontOutlineSettings:"Font Outline Settings Structure (by ref)"=@n6.OutlineSettings) -> (
    OutlineSize:int
    bMiteredCorners:bool=~
    bSeparateFillAlpha:bool=~
    bApplyOutlineToDropShadows:bool=~
    OutlineMaterial:Object Reference
    OutlineColor:Linear Color Structure
   ) {is=title K2Node_BreakStruct}
   n9 = Select(
    Option 0:Object Reference=/Game/Athena/HUD/Tricks/Materials/M_TrickText
    Option 1:Object Reference=/Game/Athena/HUD/Tricks/Materials/M_TrickText_Critical
    Index=@n1.isCritical
   ) -> (ReturnValue:Object Reference)
   > SetFont(
    self=@n5
    InFontInfo=MakeSlateFontInfo!(
     FontObject:Object Reference=@n6
     FontMaterial:Object Reference=@n6.FontMaterial
     OutlineSettings:Font Outline Settings Structure=Make Font Outline Settings!(
      OutlineSize:int=@n8
      bMiteredCorners:bool=False
      bSeparateFillAlpha:bool=@n8.bSeparateFillAlpha
      bApplyOutlineToDropShadows:bool=@n8.bApplyOutlineToDropShadows
      OutlineMaterial:Object Reference=@n9
      OutlineColor:Linear Color Structure=@n8.OutlineColor
     ) -> (FontOutlineSettings:Font Outline Settings Structure) {is=title K2Node_MakeStruct}
     TypefaceFontName:Name=@n6.TypefaceFontName
     Size:float="To Float (Integer)"(InInt=@n7)
     LetterSpacing:int=0
     SkewAmount:float=0
     bForceMonospaced:bool=false
     bMaterialIsStencil:bool=false
     MonospacedWidth:float=1
    ) -> (ReturnValue:Slate Font Info Structure)
   )
   > SetColorAndOpacity(
    self=@n5
    InColorAndOpacity=Make Slate Color(SpecifiedColor=Select(Option 0:Linear Color Structure=(R=1,G=0.9911,B=0.7011,A=1), Option 1:Linear Color Structure=(R=1,G=0.6038,B=0,A=1), Index=@n1.isCritical) -> (ReturnValue:Linear Color Structure))
   )
   > SetRenderTranslation(
    self=@n5
    Translation=Select(Option 0:Vector 2D Structure=(X=0,Y=-9), Option 1:Vector 2D Structure, Index=@n1.isCritical) -> (ReturnValue:Vector 2D Structure)
   )
   SimulateTrick() {is=title K2Node_CustomEvent}
   > TrickSequenceStarted(self:Athena Trick Feed Base Object Reference)
   > SetCurrentValue!(execute, self:Common Numeric Text Block Object Reference=Get TotalScore(), NewValue:float=0) -> (then)
   > UpdateScore!(
    execute
    self:Athena Trick Feed Base Object Reference
    TrickSet:Fort Vehicle Trick Set Object Reference
    TotalScore:int=RandomIntegerInRange!(Min:int=20000, Max:int=100000) -> (ReturnValue:int=0)
   ) -> (then)
   > AddTrick!(execute, self:Athena Trick Feed Base Object Reference, TrickName:"Text (by ref)"=MakeLiteralText(Value="LOCTEXT(\"SPIN IT\")")) -> (then)
   > UpdateMultiplier!(execute, self:Athena Trick Feed Base Object Reference, Multiplier:int=1) -> (then)
   > Delay(Duration=1)
   > AddTrick(TrickName=MakeLiteralText(Value="LOCTEXT(\"360!!!!\")"))
   > Delay(Duration=0.1)
   > AddTrick(TrickName=MakeLiteralText(Value="LOCTEXT(\"Spinzo!!!!\")"))
   > Delay(Duration=0.2)
   > AddTrick(TrickName=MakeLiteralText(Value="LOCTEXT(\"Blitzo!!!\")"))
   > UpdateMultiplier(Multiplier=2)
   > Delay(Duration=1)
   > AddTrick(TrickName=MakeLiteralText(Value="LOCTEXT(\"720!!!\")"))
   > UpdateMultiplier(Multiplier=3)
   > Delay(Duration=1)
   > TrickSequenceComplete(self:Athena Trick Feed Base Object Reference)
   n10 = ToggleCompleteStyle!() -> (OutputDelegate, then, isComplete:bool=~) {is=title K2Node_CustomEvent}
   n11 = Get TotalScore()
   > SetColorAndOpacity(
    self=@n11
    InColorAndOpacity=Select(
     Option 0:Slate Color Structure=Make Slate Color(SpecifiedColor=(R=1,G=0.9911,B=0.7011,A=1))
     Option 1:Slate Color Structure=Make Slate Color(SpecifiedColor=(R=1,G=1,B=1,A=1))
     Index=@n10.isComplete
    ) -> (ReturnValue:Slate Color Structure)
   )
   n12 = Break Slate Font Info(SlateFontInfo=Get Font(self=@n11)) -> (Size=~, SkewAmount=~, bForceMonospaced=~, bMaterialIsStencil=~, MonospacedWidth=~)
   n13 = Select(
    Option 0:int=Truncate(A=@n12.Size)
    Option 1:int=int + int(A=Truncate(A=@n12.Size), B=3) -> (ReturnValue=0) {is=title K2Node_CommutativeAssociativeBinaryOperator}
    Index=@n10.isComplete
   ) -> (ReturnValue:int)
   n14 = Break Font Outline Settings(FontOutlineSettings=@n12.OutlineSettings) -> (bMiteredCorners=~, bSeparateFillAlpha=~, bApplyOutlineToDropShadows=~)
   n15 = Select(
    Option 0:Object Reference=&M_TrickText
    Option 1:Object Reference=/Game/Athena/HUD/Tricks/Materials/M_TrickText_Complete
    Index=@n10.isComplete
   ) -> (ReturnValue:Object Reference)
   > SetFont(
    self=@n11
    InFontInfo=MakeSlateFontInfo(
     FontObject=@n12
     FontMaterial=@n12.FontMaterial
     OutlineSettings=Make Font Outline Settings(
      OutlineSize=@n14
      bMiteredCorners=False
      bSeparateFillAlpha=@n14.bSeparateFillAlpha
      bApplyOutlineToDropShadows=@n14.bApplyOutlineToDropShadows
      OutlineMaterial=@n15
      OutlineColor=@n14.OutlineColor
     )
     TypefaceFontName=@n12.TypefaceFontName
     Size="To Float (Integer)"(InInt=@n13)
     LetterSpacing=0
     SkewAmount=0
     bForceMonospaced=false
     bMaterialIsStencil=false
     MonospacedWidth=1
    )
   )
   n16 = EventTrickCompliment() {is=title K2Node_CustomEvent}
   n17 = GetCalloutText!(execute=[@n31.then, @n16.then], self) -> (then, Callout:Text)
   > n18 = Branch(Condition=TextIsEmpty(InText=@n17))
   n19 = Create Entry of Class!(
    execute=@n18.else
    self:Dynamic Entry Box Object Reference=Get TrickList() -> (TrickList:Fort Dynamic Entry Box Object Reference)
    EntryClass:User Widget Class Reference=&/CalloutEntry.CalloutEntry_C
   ) -> (then, ReturnValue:Callout Entry Object Reference)
   > SetCallout!(execute, self:Callout Entry Object Reference=@n19, InText:Text=@n17) -> (then)
   n20 = Event UpdateScore!() -> (OutputDelegate, then, TrickSet:Fort Vehicle Trick Set Object Reference, TotalScore:int=0) {is=title K2Node_Event}
   n21 = Get CritThreshold() -> (CritThreshold:int=0)
   n22 = Get isCritical() -> (isCritical:bool)
   > n23 = Branch(Condition=AND Boolean(A="integer >= integer"!(A:int=@n20.TotalScore, B:int=@n21) -> (ReturnValue:bool) {is=title K2Node_CallFunction}, B=NOT Boolean(A=@n22)))
   > n24 = CritStyleEvent!(execute, self, isCritical:bool=true) -> (then) {is=title K2Node_CallFunction}
   InterpolateToValue!(
    execute=[@n24.then, @n23.else]
    self:Common Numeric Text Block Object Reference=Get TotalScore()
    TargetValue:float="To Float (Integer)"(InInt=@n20.TotalScore)
    MaximumInterpolationDuration:float=1
    MinimumChangeRate:float=3
    OutroOffset:float=0
   ) -> (then)
   > Set isAmazing(isAmazing:bool="integer >= integer"(A=@n20.TotalScore, B=Get AmazingThreshold() -> (AmazingThreshold:int=0))) -> (Output_Get:bool)
   n25 = Event UpdateMultiplier!() -> (OutputDelegate, then, Multiplier:int=0) {is=title K2Node_Event}
   > "SetText (Text)"(
    self=Get Multiplier() -> (Multiplier:Common Text Object Reference)
    InText=Format Text!(Format="LOCTEXT(\"x{Multiplier}\")", Multiplier:int=@n25.Multiplier) -> (Result)
   )
   > PlayAnimation(InAnimation=Get MultiplierBump() -> (MultiplierBump:Widget Animation Object Reference))
   Event TrickSequenceStarted() {is=title K2Node_Event}
   > PrintString!(execute, InString=Trick Sequence: STARTED, bPrintToScreen, bPrintToLog, TextColor, Duration, Key) -> (then)
   > SetVisibility(InVisibility=HitTestInvisible)
   > "SetText (Text)"(self=Get TrickStats() -> (TrickStats:Common Text Object Reference))
   > "SetText (Text)"(self=Get Multiplier())
   > SetCurrentValue(self=Get TotalScore(), NewValue=0)
   > Reset!(execute, self:Dynamic Entry Box Object Reference=Get TrickList(), bDeleteWidgets:bool=false) -> (then)
   > Clear and Invalidate Timer by Handle(Handle=Get RemoveInfoTimer() -> (RemoveInfoTimer:Timer Handle Structure))
   > CritStyleEvent(isCritical=false)
   > ToggleCompleteStyle!(execute, self, isComplete:bool=false) -> (then) {is=title K2Node_CallFunction}
   > SetVisibility(self=Get Image_Glow(), InVisibility=Collapsed)
   > SetVisibility(self=Get FailedText() -> (FailedText:Common Text Object Reference), InVisibility=Collapsed)
   > SetRenderOpacity(self=Get TrickInformation() -> (TrickInformation:Vertical Box Object Reference), InOpacity=1)
   > PlayAnimation(InAnimation=Get TrickReset() -> (TrickReset:Widget Animation Object Reference))
   > StartTrickMusic()
   Event TrickSequenceFailed() {is=title K2Node_Event}
   > PrintString(InString=Trick Sequence Failed)
   n26 = Get WipeoutDescription() -> (WipeoutDescription:Array of Texts)
   n27 = RandomIntegerInRange(Min=0, Max=int - int(A=Length(TargetArray:Array of Texts=@n26), B=1) {is=title K2Node_CallFunction}) -> (ReturnValue=0)
   n28 = Get FailedText()
   > "SetText (Text)"(
    self=@n28
    InText=TextToUpper!(InText:"Text (by ref)"="Get (a copy)"(Array:Array of Texts=@n26, Dimension 1=@n27) -> (Output:Text)) -> (ReturnValue:Text)
   )
   > SetVisibility(self=@n28, InVisibility=HitTestInvisible)
   > SetRenderOpacity(self=Get TrickInformation(), InOpacity=1)
   > PlayAnimation(InAnimation=Get TrickFailed() -> (TrickFailed:Widget Animation Object Reference), PlaybackSpeed=1.5)
   > StopTrickMusic!(execute, self, Success:bool=false) -> (then)
   Event TrickSequenceComplete() {is=title K2Node_Event}
   > PrintString(InString=Trick Sequence Complete)
   > n29 = Sequence()
   n30 = Get Image_Glow()
   > SetVisibility(self=@n30, InVisibility=HitTestInvisible)
   > SetBrushTintColor(self=@n30, TintColor=Make Slate Color(SpecifiedColor=(R=1,G=0.0296,B=0,A=1)))
   > PlayAnimation(InAnimation=Get TriangleFlare() -> (TriangleFlare:Widget Animation Object Reference))
   > n31 = ToggleCompleteStyle(isComplete=true)
   StopTrickMusic(execute=@n29.then_1, Success=true)
   > n32 = Set Timer by Event(Delegate=Create Event(), Time=3.5, bLooping=false)
   > Set RemoveInfoTimer(RemoveInfoTimer:Timer Handle Structure=@n32) -> (Output_Get:Timer Handle Structure)
   n33 = Event PreConstruct()
   > Branch(Condition=@n33.IsDesignTime)
   > CritStyleEvent(isCritical=Get DEVCritDisplay() -> (DEVCritDisplay:bool))
   > ToggleCompleteStyle(isComplete=Get DEVFinalDisplay() -> (DEVFinalDisplay:bool))
   Event TrickSequenceCanceled() {is=title K2Node_Event}
   > PrintString(InString=Trick Sequence Canceled)
   n34 = Get FailedText()
   > "SetText (Text)"(self=@n34, InText="LOCTEXT(\"BAILED!\")")
   > SetVisibility(self=@n34, InVisibility=HitTestInvisible)
   > SetRenderOpacity(self=Get TrickInformation(), InOpacity=1)
   > PlayAnimation(InAnimation=Get TrickFailed(), PlaybackSpeed=2)
   > StopTrickMusic(Success=false)
   n35 = Event AddTrick!() -> (OutputDelegate, then, TrickName:"Text (by ref)") {is=title K2Node_Event}
   > n36 = Create Entry!(execute, self:Dynamic Entry Box Object Reference=Get TrickList()) -> (then, ReturnValue:User Widget Object Reference)
   > n37 = Cast To TrickEntry!(execute, Object=@n36) -> (then, CastFailed, AsTrick Entry:Trick Entry Object Reference) {is=title K2Node_DynamicCast}
   > SetTrick!(execute, self:Trick Entry Object Reference=@n37, InText:Text=@n35.TrickName) -> (then)
   > PlayTrickStinger()
   n38 = Event UpdateTrickStats!() -> (OutputDelegate, then, StatsText:"Text (by ref)") {is=title K2Node_Event}
   > "SetText (Text)"(self=Get TrickStats(), InText=@n38.StatsText)
   "Animation Finished (TrickFailed)"!() -> (OutputDelegate:Delegate, then:Exec) {is=title K2Node_WidgetAnimationEvent}
   > HIdeTrickUI() {is=title K2Node_CallFunction}
   "Animation Finished (TrickComplete)"!() -> (OutputDelegate:Delegate, then:Exec) {is=title K2Node_WidgetAnimationEvent}
   > SetVisibility(InVisibility=Hidden)
   > RequestNextSequence(self:Athena Trick Feed Base Object Reference)
   > PrintString(InString=Request Next Sequence)
  ]}
 GetCalloutText {nodes [
   GetCalloutText() {is=title K2Node_FunctionEntry}
   n1 = Select(
    Option 0:Array of Texts=Make Array("[0]":Text) -> (Array:Array of Texts)
    Option 1:Array of Texts=Select(
     Option 0=Get NiceTricksDescription() -> (NiceTricksDescription:Array of Texts)
     Option 1=Get AmazingTricksDescription() -> (AmazingTricksDescription:Array of Texts)
     Index=Get isAmazing() -> (isAmazing:bool)
    )
    Index=Get isCritical()
   ) -> (ReturnValue:Array of Texts)
   n2 = int - int(A=Length(TargetArray=@n1), B=1)
   > Return Node!(execute, Callout:Text="Get (a copy)"(Array=@n1, Dimension 1=RandomIntegerInRange(Min=0, Max=@n2) -> (ReturnValue=0)))
   Get OkTricksDescription() -> (OkTricksDescription:Array of Texts)
  ]}
 PlayTrickStinger {nodes [
   PlayTrickStinger() {is=title K2Node_FunctionEntry}
   > n1 = Get!(execute:Exec) -> (then:Exec, else:Exec, TrickSequenceMusic:Audio Component Object Reference)
   > Execute Trigger Parameter!(execute, self:Audio Parameter Controller Interface Interface=@n1, InName:Name=OnTrick) -> (then)
  ]}
 StartTrickMusic {nodes [
   StartTrickMusic() {is=title K2Node_FunctionEntry}
   > RegisterDefaultSubmix(self:Audio Analysis Subsystem Object Reference=Get FortAudioAnalysisSubsystem() -> (ReturnValue:Audio Analysis Subsystem Object Reference) {is=title K2Node_GetSubsystem})
   > n1 = Cast To FortGameStateAthena!(execute, Object=GetGameState!() -> (ReturnValue:Game State Base Object Reference)) -> (then, CastFailed, AsFort Game State Athena:Fort Game State Athena Object Reference) {is=title K2Node_DynamicCast}
   n2 = GetAthenaPlaylistContextTags!(self:Fort Game State Athena Object Reference=@n1) -> (ReturnValue:Gameplay Tag Container Structure)
   > n3 = SpawnSound2D!(
    execute
    Sound:Sound Base Object Reference=Select(
     Option 0:Sound Base Object Reference=Get TrickMusicMetaSound() -> (TrickMusicMetaSound:Sound Base Object Reference)
     Option 1:Sound Base Object Reference=Get TrickMusicMetaSoundFigment() -> (TrickMusicMetaSoundFigment:Sound Base Object Reference)
     Index=HasTag!(
      TagContainer:"Gameplay Tag Container Structure (by ref)"=@n2
      Tag:Gameplay Tag Structure=(TagName="Athena.Playlist.Figment")
      bExactMatch:bool=false
     ) -> (ReturnValue:bool)
    ) -> (ReturnValue:Sound Base Object Reference)
    VolumeMultiplier:float=1
    PitchMultiplier:float=1
    StartTime:float=0
    ConcurrencySettings:Sound Concurrency Object Reference
    bPersistAcrossLevelTransition:bool=false
    bAutoDestroy:bool=true
   ) -> (then, ReturnValue:Audio Component Object Reference)
   > Set TrickSequenceMusic(TrickSequenceMusic:Audio Component Object Reference=@n3) -> (Output_Get:Audio Component Object Reference)
   n4 = SpawnSound2D(
    execute=@n1.CastFailed
    Sound=Get TrickMusicMetaSound()
    VolumeMultiplier=1
    PitchMultiplier=1
    StartTime=0
    bPersistAcrossLevelTransition=false
    bAutoDestroy=true
   )
   > Set TrickSequenceMusic(TrickSequenceMusic=@n4)
  ]}
 StopTrickMusic {nodes [
   StopTrickMusic!() -> (then, Success:bool=~) {is=title K2Node_FunctionEntry}
   > UnRegisterDefaultSubmix(self:Audio Analysis Subsystem Object Reference=Get FortAudioAnalysisSubsystem())
   > n1 = Cast To FortGameStateAthena(Object=GetGameState())
   > n2 = Branch(Condition=HasTag(TagContainer=GetAthenaPlaylistContextTags(self=@n1), Tag=(TagName="Athena.Playlist.Figment"), bExactMatch=true))
   > n3 = Get()
   > Set Object Parameter!(
    execute
    self:Audio Parameter Controller Interface Interface=@n3
    InName:Name=MusicFinished
    InValue:Object Reference=Select(
     Option 0=/Game/Sounds/Fort_UI_Sounds/Gameplay/Tricks/TrickUI_SequenceEnd_Fail_08
     Option 1=/Game/Sounds/Fort_UI_Sounds/Gameplay/Tricks/TrickUI_SequenceEnd_Success_07
     Index=Get Success() -> (Success:bool)
    )
   ) -> (then) {is=title K2Node_CallFunction}
   > Execute Trigger Parameter(self=@n3, InName=OnFinished)
   n4 = Get(execute=[@n1.CastFailed, @n2.else])
   > Set Object Parameter(
    self=@n4
    InName=MusicFinished
    InValue=Select(
     Option 0=/Game/Sounds/Fort_UI_Sounds/Gameplay/Tricks/V2/Waves/MX_Trick_MC_DB_Stinger_Fail_01
     Option 1=/Game/Sounds/Fort_UI_Sounds/Gameplay/Tricks/V2/Waves/MX_Trick_MC_DB_Stinger_Success_03
     Index=Get Success()
    )
   )
   > Execute Trigger Parameter(self=@n4, InName=OnFinished)
  ]}
]
variables [
 AmazingThreshold {type=integer, default=200000}
 AmazingTricksDescription {type="array<string>", default ["Extreme!!", "Wicked!!", "Amazing!!", "Savage!!"]}
 CritThreshold {type=integer, default=100000}
 DEVCritDisplay {type=boolean}
 DEVFinalDisplay {type=boolean}
 isAmazing {type=boolean}
 isCritical {type=boolean}
 NiceTricksDescription {type="array<string>", default ["Nice!", "Hot!", "Whoa!", "Great!"]}
 OkTricksDescription {type="array<string>", default [Ok, "Meh!", "So-So!", Eh...]}
 RemoveInfoTimer {type=TimerHandle}
 TrickMusicMetaSound {type=SoundBase, default=/Game/Sounds/Fort_UI_Sounds/Gameplay/Tricks/V2/MSS_UI_TrickMusic_Loop}
 TrickMusicMetaSoundFigment {type=SoundBase, default=/Game/Sounds/Fort_UI_Sounds/Gameplay/Tricks/V2/MSS_UI_TrickMusic_Loop_Figment}
 TrickSequenceLayer {type=AudioComponent, default=null}
 TrickSequenceMusic {type=AudioComponent, default=null}
 WipeoutDescription {type="array<string>", default ["Wipeout!", "Bailed!", "Biffed It!", "Busted!", "Ouch!", "Wasted!"]}
]
widgets {
 info {inheritedWidgetCount=1, parentClass=/Script/FortniteUI.AthenaTrickFeedBase, rootWidgetClass=&CanvasPanel, widgetCount=26}
 widgets [
  CanvasPanel_0 {widgetClassPath=&CanvasPanel, widget=CanvasPanel_0}
  CommonBorder_TrickContainer {
   widgetClassPath=&CommonBorder
   parent=CanvasPanel_0
   widget=CommonBorder_TrickContainer
   properties {padding=<ltrb 0 0 0 0>}
   slot {
    class=CanvasPanelSlot
    properties {bAutoSize=true, layoutData {alignment=<1 1>, anchors {maximum=<0.5 0.5>, minimum=<0.5 0.5>}, offsets=<ltrb -350 250 100 30>}}
   }
  }
  TrickInformation {
   widgetClassPath=&VerticalBox
   bIsVariable=true
   parent=CommonBorder_TrickContainer
   widget=TrickInformation
   properties {renderTransformPivot=<1 1>}
   slot {class=BorderSlot, properties {padding=<ltrb 0 0 0 0>}}
  }
  TrickList {
   widgetClassPath=/Script/FortniteUI.FortDynamicEntryBox
   bIsVariable=true
   parent=TrickInformation
   widget=TrickList
   properties {
    entryBoxType=Vertical
    entryWidgetClass=&/TrickEntry.TrickEntry_C
    maximumEntries=4
    numDesignerPreviewEntries=8
    patternDirection=LastToFirst
    renderTransformPattern [
     {angle=0, scale=<1 1>, shear=<0 0>, translation=<0 0>}
     {angle=0, scale=<1 1>, shear=<0 0>, translation=<8 0>}
     {angle=0, scale=<1 1>, shear=<0 0>, translation=<16 0>}
     {angle=0, scale=<1 1>, shear=<0 0>, translation=<24 0>}
    ]
   }
   slot {class=VerticalBoxSlot, properties {horizontalAlignment=HAlign_Right}}
  }
  Overlay_0 {
   widgetClassPath=&Overlay
   parent=TrickInformation
   widget=Overlay_0
   slot {class=VerticalBoxSlot, properties {horizontalAlignment=HAlign_Right, verticalAlignment=VAlign_Center, padding=<ltrb 0 0 15 0>}}
  }
  HorizontalBox_0 {
   widgetClassPath=&HorizontalBox
   parent=Overlay_0
   widget=HorizontalBox_0
   properties {renderTransform {angle=0, scale=<1 1>, shear=<0 0>, translation=<0 6>}}
   slot {class=OverlaySlot, properties {horizontalAlignment=HAlign_Right}}
  }
  OverlayScore+effects {
   widgetClassPath=&Overlay
   parent=HorizontalBox_0
   widget=OverlayScore+effects
   properties {renderTransform {angle=0, scale=<1 1>, shear=<0 0>, translation=<0 -9>}}
   slot {class=HorizontalBoxSlot, properties {horizontalAlignment=HAlign_Right, verticalAlignment=VAlign_Bottom, padding=<ltrb 0 -15 0 -10>}}
  }
  CommonBorder_Flare {
   widgetClassPath=&CommonBorder
   parent=OverlayScore+effects
   widget=CommonBorder_Flare
   properties {padding=<ltrb 0 0 0 0>}
   slot {class=OverlaySlot, properties {horizontalAlignment=HAlign_Fill, verticalAlignment=VAlign_Fill}}
  }
  SizeBox_Flares {
   widgetClassPath=&SizeBox
   parent=CommonBorder_Flare
   widget=SizeBox_Flares
   properties {bOverride_HeightOverride=true, renderOpacity=0, widthOverride=381.5838}
   slot {class=BorderSlot, properties {padding=<ltrb 0 0 0 0>}}
  }
  HorizontalBox_2 {widgetClassPath=&HorizontalBox, parent=SizeBox_Flares, widget=HorizontalBox_2, slot {class=SizeBoxSlot}}
  ImageTriangleL {
   widgetClassPath=&Image
   parent=HorizontalBox_2
   widget=ImageTriangleL
   properties {
    brush {
     bIsDynamicallyLoaded=false
     drawAs=Image
     imageType=NoImage
     mirroring=NoMirror
     resourceName=None
     resourceObject=/Game/UI/HUD/Materials/M_FX_TriangleSpiral
     tiling=NoTile
     imageSize=<100 10>
     margin=<ltrb 0 0 0 0>
     outlineSettings {
      bUseBrushTransparency=false
      roundingType=HalfHeightRadius
      width=0
      color {colorUseRule=UseColor_Specified, specifiedColor=<rgba 0 0 0 0>}
      cornerRadii=<0 0 0 0>
     }
     tintColor {colorUseRule=UseColor_Specified, specifiedColor=<rgba 1 0.5271 0 1>}
     uVRegion {bIsValid=false, max=<0 0>, min=<0 0>}
    }
    renderTransform {angle=0, scale=<-1 1>, shear=<0 0>, translation=<5 0>}
   }
   slot {class=HorizontalBoxSlot, properties {horizontalAlignment=HAlign_Left, padding=<ltrb -100 10 -100 10>, size {sizeRule=Fill, value=1}}}
  }
  Spacer_1 {
   widgetClassPath=&Spacer
   parent=HorizontalBox_2
   widget=Spacer_1
   slot {class=HorizontalBoxSlot, properties {size {sizeRule=Fill, value=1}}}
  }
  ImageTriangleR {
   widgetClassPath=&Image
   parent=HorizontalBox_2
   widget=ImageTriangleR
   properties {
    brush {
     bIsDynamicallyLoaded=false
     drawAs=Image
     imageType=NoImage
     mirroring=NoMirror
     resourceName=None
     resourceObject=&M_FX_TriangleSpiral
     tiling=NoTile
     imageSize=<100 10>
     margin=<ltrb 0 0 0 0>
     outlineSettings {
      bUseBrushTransparency=false
      roundingType=HalfHeightRadius
      width=0
      color {colorUseRule=UseColor_Specified, specifiedColor=<rgba 0 0 0 0>}
      cornerRadii=<0 0 0 0>
     }
     tintColor {colorUseRule=UseColor_Specified, specifiedColor=<rgba 1 0.5271 0 1>}
     uVRegion {bIsValid=false, max=<0 0>, min=<0 0>}
    }
    renderTransform {angle=0, scale=<1 -1>, shear=<0 0>, translation=<-7 0>}
   }
   slot {class=HorizontalBoxSlot, properties {horizontalAlignment=HAlign_Right, padding=<ltrb -100 10 -100 10>, size {sizeRule=Fill, value=1}}}
  }
  TotalScore {
   widgetClassPath=/Script/CommonUI.CommonNumericTextBlock
   bIsVariable=true
   parent=OverlayScore+effects
   widget=TotalScore
   properties {
    currentNumericValue=52348
    justification=Right
    text=0
    colorAndOpacity {colorUseRule=UseColor_Specified, specifiedColor=<rgba 1 0.9911 0.7011 1>}
    font {
     bForceMonospaced=false
     bMaterialIsStencil=false
     fontMaterial=null
     fontObject=&Burbank
     letterSpacing=0
     monospacedWidth=1
     size=30
     skewAmount=0
     typefaceFontName=Burbank-PlayerPower
     outlineSettings {
      bApplyOutlineToDropShadows=false
      bMiteredCorners=false
      bSeparateFillAlpha=false
      outlineMaterial=&M_TrickText
      outlineSize=1
      outlineColor=<rgba 1 1 1 1>
     }
    }
    renderTransform {angle=0, scale=<1 1>, shear=<-10 0>, translation=<0 0>}
    shadowColorAndOpacity=<rgba 0.0156 0.0082 0.0063 0.8>
    shadowOffset=<5 5>
   }
   slot {class=OverlaySlot, properties {horizontalAlignment=HAlign_Right, verticalAlignment=VAlign_Bottom, padding=<ltrb 0 0 0 -8>}}
  }
  Multiplier {
   widgetClassPath=&CommonTextBlock
   bIsVariable=true
   parent=HorizontalBox_0
   widget=Multiplier
   properties {
    text=x2
    font {
     bForceMonospaced=false
     bMaterialIsStencil=false
     fontMaterial=null
     fontObject=&Roboto
     letterSpacing=0
     monospacedWidth=1
     size=26
     skewAmount=0
     typefaceFontName=Bold
     outlineSettings {
      bApplyOutlineToDropShadows=false
      bMiteredCorners=false
      bSeparateFillAlpha=false
      outlineMaterial=null
      outlineSize=1
      outlineColor=<rgba 0.6584 0.1746 0.0137 1>
     }
    }
    renderTransform {angle=0, scale=<1 1>, shear=<-10 0>, translation=<0 -6>}
    renderTransformPivot=<0.2 0.8>
   }
   slot {class=HorizontalBoxSlot, properties {verticalAlignment=VAlign_Bottom}}
  }
  Image_Glow {
   widgetClassPath=&Image
   bIsVariable=true
   parent=Overlay_0
   widget=Image_Glow
   properties {
    brush {
     bIsDynamicallyLoaded=false
     drawAs=Image
     imageType=NoImage
     mirroring=NoMirror
     resourceName=None
     resourceObject=/Game/Athena/HUD/Materials/M_CountdownBG_SolidGlow
     tiling=NoTile
     imageSize=<170.1995 93.5933>
     margin=<ltrb 0 0 0 0>
     outlineSettings {
      bUseBrushTransparency=false
      roundingType=HalfHeightRadius
      width=0
      color {colorUseRule=UseColor_Specified, specifiedColor=<rgba 0 0 0 0>}
      cornerRadii=<0 0 0 0>
     }
     tintColor {colorUseRule=UseColor_Specified, specifiedColor=<rgba 1 0.3046 0 1>}
     uVRegion {bIsValid=false, max=<0 0>, min=<0 0>}
    }
    renderTransform {angle=0, scale=<1 1>, shear=<-10 0>, translation=<2 0>}
   }
   slot {class=OverlaySlot, properties {horizontalAlignment=HAlign_Fill, verticalAlignment=VAlign_Fill, padding=<ltrb -100 -50 -100 -50>}}
  }
  TrickStats {
   widgetClassPath=&CommonTextBlock
   bIsVariable=true
   parent=TrickInformation
   widget=TrickStats
   properties {
    text="← 30m ↑ 30m | 4.2s"
    font {
     bForceMonospaced=false
     bMaterialIsStencil=false
     fontMaterial=null
     fontObject=&NotoSans
     letterSpacing=0
     monospacedWidth=1
     size=18
     skewAmount=0
     typefaceFontName=NotoSans-Regular
     outlineSettings {
      bApplyOutlineToDropShadows=false
      bMiteredCorners=false
      bSeparateFillAlpha=false
      outlineMaterial=null
      outlineSize=1
      outlineColor=<rgba 0.6584 0.1746 0.0137 1>
     }
    }
    renderTransform {angle=0, scale=<1 1>, shear=<-10 0>, translation=<0 0>}
   }
   slot {class=VerticalBoxSlot, properties {horizontalAlignment=HAlign_Right, verticalAlignment=VAlign_Center, padding=<ltrb 0 0 30 0>}}
  }
  RichTrickStats {
   widgetClassPath=/Script/UMG.RichTextBlock
   bIsVariable=true
   parent=TrickInformation
   widget=RichTrickStats
   slot {class=VerticalBoxSlot, properties {padding=<ltrb 0 0 30 0>}}
  }
  Overlay_WipeoutContainer {
   widgetClassPath=&Overlay
   parent=CanvasPanel_0
   widget=Overlay_WipeoutContainer
   properties {renderOpacity=0}
   slot {
    class=CanvasPanelSlot
    properties {
     bAutoSize=true
     zOrder=1
     layoutData {alignment=<0.5 0.5>, anchors {maximum=<0.5 0.5>, minimum=<0.5 0.5>}, offsets=<ltrb -458.1937 148.8222 100 30>}
    }
   }
  }
  SizeBox_0 {
   widgetClassPath=&SizeBox
   parent=Overlay_WipeoutContainer
   widget=SizeBox_0
   properties {bOverride_WidthOverride=true, maxDesiredWidth=145.9634, widthOverride=250}
   slot {class=OverlaySlot}
  }
  ScaleBox_0 {widgetClassPath=&ScaleBox, parent=SizeBox_0, widget=ScaleBox_0, properties {stretchDirection=DownOnly}, slot {class=SizeBoxSlot}}
  Overlay_3 {widgetClassPath=&Overlay, parent=ScaleBox_0, widget=Overlay_3, slot {class=ScaleBoxSlot}}
  Image_WipeoutBG {
   widgetClassPath=&Image
   parent=Overlay_3
   widget=Image_WipeoutBG
   properties {brush {
     bIsDynamicallyLoaded=false
     drawAs=NoDrawType
     imageType=NoImage
     mirroring=NoMirror
     resourceName=None
     resourceObject=/Game/Athena/HUD/Tricks/Materials/M_UI_TrickBailedBG_Inst
     tiling=NoTile
     imageSize=<32 32>
     margin=<ltrb 0 0 0 0>
     outlineSettings {
      bUseBrushTransparency=false
      roundingType=HalfHeightRadius
      width=0
      color {colorUseRule=UseColor_Specified, specifiedColor=<rgba 0 0 0 0>}
      cornerRadii=<0 0 0 0>
     }
     tintColor {colorUseRule=UseColor_Specified, specifiedColor=<rgba 1 1 1 1>}
     uVRegion {bIsValid=false, max=<0 0>, min=<0 0>}
    }}
   slot {class=OverlaySlot, properties {horizontalAlignment=HAlign_Fill, verticalAlignment=VAlign_Fill, padding=<ltrb -20 -10 -40 -50>}}
  }
  Image_WipeoutPlateBG {
   widgetClassPath=&Image
   parent=Overlay_3
   widget=Image_WipeoutPlateBG
   properties {
    brush {
     bIsDynamicallyLoaded=false
     drawAs=Image
     imageType=NoImage
     mirroring=NoMirror
     resourceName=None
     resourceObject=/Game/Athena/HUD/Tricks/Materials/M_UI_Trick_BailedBG_Inst
     tiling=NoTile
     imageSize=<64 0>
     margin=<ltrb 0 0 0 0>
     outlineSettings {
      bUseBrushTransparency=false
      roundingType=HalfHeightRadius
      width=0
      color {colorUseRule=UseColor_Specified, specifiedColor=<rgba 0 0 0 0>}
      cornerRadii=<0 0 0 0>
     }
     tintColor {colorUseRule=UseColor_Specified, specifiedColor=<rgba 1 1 1 1>}
     uVRegion {bIsValid=false, max=<0 0>, min=<0 0>}
    }
    colorAndOpacity=<rgba 0.5 0 0.0845 1>
    renderTransform {angle=-7, scale=<1 1>, shear=<-33 0>, translation=<0 12>}
   }
   slot {class=OverlaySlot, properties {horizontalAlignment=HAlign_Fill, verticalAlignment=VAlign_Fill, padding=<ltrb -50 5 -50 5>}}
  }
  FailedText {
   widgetClassPath=&CommonTextBlock
   bIsVariable=true
   parent=Overlay_3
   widget=FailedText
   properties {
    text=OUCH
    colorAndOpacity {colorUseRule=UseColor_Specified, specifiedColor=<rgba 1 0.9341 0.8498 1>}
    font {
     bForceMonospaced=false
     bMaterialIsStencil=false
     fontMaterial=null
     fontObject=&Burbank
     letterSpacing=0
     monospacedWidth=1
     size=45
     skewAmount=0
     typefaceFontName=Burbank-Black
     outlineSettings {
      bApplyOutlineToDropShadows=false
      bMiteredCorners=false
      bSeparateFillAlpha=false
      outlineMaterial=/Game/Athena/HUD/Tricks/Materials/M_TrickFailedOutline
      outlineSize=1
      outlineColor=<rgba 1 1 1 1>
     }
    }
    renderTransform {angle=-6.9149, scale=<1 1>, shear=<-4 0>, translation=<0 0>}
    shadowColorAndOpacity=<rgba 0.02 0.0036 0 0.8>
   }
   slot {class=OverlaySlot}
  }
  TutorialHighlight {widgetClassPath=&TutorialHighlight}
 ]
}
